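# Physics asset for the belt on the female medium StallionAviator character body, with capsule bodies for a dynamic belt chain and pouch.
# Belt segments 01 to 07 chain from spine_01 with free swing and locked twist, while segment 08 is angularly locked to 07.
# The pouch is angularly locked to spine_01.
# Pelvis, lower spine and left leg supply body shapes, including a large offset pelvis box; spine_02 has no constraint.
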PhysicsAsset /BRCosmetics/Characters/Player/Female/Medium/Bodies/F_MED_StallionAviator/Meshes/F_MED_StallionAviator_Belt_Physics
bodies {
 calf_l [{
   shapeType=Capsule
   length=31.396
   radius=7.9713
   shapeName=calf_l_capsule
   center=<-17.7438 -2.396 -0.6179>
   rotation=<rot 0 98.502 -88.0122>
  }]
 dyn_belt_01 [{
   shapeType=Capsule
   length=0.5
   radius=1
   shapeName=dyn_belt_01_capsule
   center=<0.4241 0.0376 0.2439>
   rotation=<rot -4.112 -89.9999 -101.7746>
  }]
 dyn_belt_02 [{
   shapeType=Capsule
   length=1.3371
   radius=1.2576
   shapeName=dyn_belt_02_capsule
   center=<0.8219 0.0812 -0.0342>
   rotation=<rot -0.2967 -90 -111.7728>
  }]
 dyn_belt_03 [{
   shapeType=Capsule
   length=2.037
   radius=1.4958
   shapeName=dyn_belt_03_capsule
   center=<1.7192 -0.2846 0.1593>
   rotation=<rot 0 94.7252 -89.1713>
  }]
 dyn_belt_04 [{
   shapeType=Capsule
   length=2.841
   radius=1.6542
   shapeName=dyn_belt_04_capsule
   center=<3.1663 -0.1897 0.1026>
   rotation=<rot 0 89.5324 -93.1769>
  }]
 dyn_belt_05 [{
   shapeType=Capsule
   length=2.3418
   radius=1.4993
   shapeName=dyn_belt_05_capsule
   center=<1.9977 -0.4471 0.2325>
   rotation=<rot 0 96.7324 -92.034>
  }]
 dyn_belt_06 [{
   shapeType=Capsule
   length=2.3509
   radius=1.5872
   shapeName=dyn_belt_06_capsule
   center=<1.3006 -0.3825 0.0841>
   rotation=<rot 0 95.5123 -84.2561>
  }]
 dyn_belt_07 [{
   shapeType=Capsule
   length=3.2121
   radius=1.6394
   shapeName=dyn_belt_07_capsule
   center=<0.9411 -0.1876 0.0423>
   rotation=<rot 0 -86.1443 -89.9558>
  }]
 dyn_belt_08 [{
   shapeType=Capsule
   length=2.1992
   radius=1.9467
   shapeName=dyn_belt_08_capsule
   center=<-0.8542 0.11 0.1067>
   rotation=<rot 21.0128 -94.147 -87.1873>
  }]
 dyn_pouch [{
   shapeType=Capsule
   length=4.2429
   radius=4.5837
   shapeName=dyn_pouch_capsule
   center=<-0.4726 -3.0174 -1.7503>
   rotation=<rot 9.2984 -90 1.542>
  }]
 pelvis [
  {shapeType=Capsule, length=14.2379, radius=8.7465, shapeName=pelvis_capsule, center=<1.1935 -0.6159 -0.1313>, rotation=<rot 0 -90 0>}
  {shapeType=Box, extentX=99.5, extentY=60.5, extentZ=25, shapeName=pelvis_box, center=<-22.5178 0 9.8542>, rotation=<rot 0 -0.0003 0>}
 ]
 spine_01 [{shapeType=Capsule, length=15.75, radius=6.505, shapeName=spine_01_capsule, center=<4.5206 0 0>, rotation=<rot 0 -90 0>}]
 spine_02 [{shapeType=Capsule, length=13.5, radius=11.5, shapeName=spine_02_capsule, rotation=<rot 90 35.8981 35.8983>}]
 thigh_l [{shapeType=Capsule, length=21.5, radius=9.75, shapeName=thigh_l_capsule, center=<-16.5215 -0.9469 -0.163>, rotation=<rot 90 179.789 179.7891>}]
}
constraints [
 {
  bone1Name=spine_01
  bone2Name=pelvis
  swing1LimitDegrees=45
  swing1Motion=Locked
  swing2LimitDegrees=45
  swing2Motion=Locked
  twistLimitDegrees=45
  twistMotion=Locked
 }
 {
  bone1Name=dyn_pouch
  bone2Name=spine_01
  swing1LimitDegrees=5
  swing1Motion=Locked
  swing2LimitDegrees=5
  swing2Motion=Locked
  twistLimitDegrees=5
  twistMotion=Locked
 }
 {
  bone1Name=thigh_l
  bone2Name=pelvis
  swing1LimitDegrees=45
  swing1Motion=Locked
  swing2LimitDegrees=45
  swing2Motion=Locked
  twistLimitDegrees=45
  twistMotion=Locked
 }
 {
  bone1Name=calf_l
  bone2Name=thigh_l
  swing1LimitDegrees=45
  swing1Motion=Locked
  swing2LimitDegrees=45
  swing2Motion=Locked
  twistLimitDegrees=45
  twistMotion=Locked
 }
 {
  bone1Name=dyn_belt_08
  bone2Name=dyn_belt_07
  swing1LimitDegrees=20
  swing1Motion=Locked
  swing2LimitDegrees=20
  swing2Motion=Locked
  twistLimitDegrees=45
  twistMotion=Locked
 }
 {
  bone1Name=dyn_belt_01
  bone2Name=spine_01
  swing1LimitDegrees=45
  swing1Motion=Free
  swing2LimitDegrees=45
  swing2Motion=Free
  twistLimitDegrees=45
  twistMotion=Locked
 }
 {
  bone1Name=dyn_belt_02
  bone2Name=dyn_belt_01
  swing1LimitDegrees=45
  swing1Motion=Free
  swing2LimitDegrees=45
  swing2Motion=Free
  twistLimitDegrees=45
  twistMotion=Locked
 }
 {
  bone1Name=dyn_belt_03
  bone2Name=dyn_belt_02
  swing1LimitDegrees=45
  swing1Motion=Free
  swing2LimitDegrees=45
  swing2Motion=Free
  twistLimitDegrees=45
  twistMotion=Locked
 }
 {
  bone1Name=dyn_belt_04
  bone2Name=dyn_belt_03
  swing1LimitDegrees=45
  swing1Motion=Free
  swing2LimitDegrees=45
  swing2Motion=Free
  twistLimitDegrees=45
  twistMotion=Locked
 }
 {
  bone1Name=dyn_belt_05
  bone2Name=dyn_belt_04
  swing1LimitDegrees=45
  swing1Motion=Free
  swing2LimitDegrees=45
  swing2Motion=Free
  twistLimitDegrees=45
  twistMotion=Locked
 }
 {
  bone1Name=dyn_belt_06
  bone2Name=dyn_belt_05
  swing1LimitDegrees=45
  swing1Motion=Free
  swing2LimitDegrees=45
  swing2Motion=Free
  twistLimitDegrees=45
  twistMotion=Locked
 }
 {
  bone1Name=dyn_belt_07
  bone2Name=dyn_belt_06
  swing1LimitDegrees=45
  swing1Motion=Free
  swing2LimitDegrees=45
  swing2Motion=Free
  twistLimitDegrees=45
  twistMotion=Locked
 }
]
