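# Physics asset for the jacket zipper on the female medium StallionAviator player cosmetic body, giving the zipper pull simulated secondary motion.
# A small capsule on the dyn_jacketZipper bone is joined by a constraint to a larger torso capsule on spine_05.
# The constraint sets narrow swing and twist angle limits, with twist the tightest, keeping the zipper's dangle close to the chest.
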PhysicsAsset /BRCosmetics/Characters/Player/Female/Medium/Bodies/F_MED_StallionAviator/Meshes/F_MED_StallionAviator_JacketZipper_Physics
bodies {
 dyn_jacketZipper [{
   shapeType=Capsule
   length=0.505
   radius=0.505
   shapeName=dyn_jacketZipper_capsule
   center=<0.0047 0.6167 -1.0006>
   rotation=<rot -30.4808 -90 -0.6167>
  }]
 spine_05 [{shapeType=Capsule, length=14.2401, radius=6.3462, shapeName=spine_05_capsule, center=<3.4909 0.8601 0>}]
}
constraints [{bone1Name=dyn_jacketZipper, bone2Name=spine_05, swing1LimitDegrees=15, swing2LimitDegrees=20, twistLimitDegrees=5}]
