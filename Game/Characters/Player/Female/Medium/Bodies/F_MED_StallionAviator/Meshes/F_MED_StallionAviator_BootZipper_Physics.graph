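# Physics asset for the boot zippers on the female medium Stallion Aviator player body, providing secondary motion for the dangling pulls.
# Each side has a small capsule on its dyn_bootZipper bone and a larger capsule on calf_twist_01, which stands in for the lower leg.
# A constraint ties each zipper to its calf with limited swing and a tighter twist limit, keeping the two sides independent.
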PhysicsAsset /Game/Characters/Player/Female/Medium/Bodies/F_MED_StallionAviator/Meshes/F_MED_StallionAviator_BootZipper_Physics
bodies {
 calf_twist_01_l [{
   shapeType=Capsule
   length=13.343
   radius=4.8996
   shapeName=calf_twist_01_l_capsule
   center=<2.4619 0.0824 0.3345>
   rotation=<rot -5 84.9998 85>
  }]
 calf_twist_01_r [{
   shapeType=Capsule
   length=12.8599
   radius=5.4761
   shapeName=calf_twist_01_r_capsule
   center=<-1.9078 -0.4837 -0.1998>
   rotation=<rot 0 84.9999 85>
  }]
 dyn_bootZipper_l [{
   shapeType=Capsule
   length=3.255
   radius=1.005
   shapeName=dyn_bootZipper_l_capsule
   center=<0.01 -0.6075 -1.5853>
   rotation=<rot -0.2786 -90 -0.6534>
  }]
 dyn_bootZipper_r [{
   shapeType=Capsule
   length=3.255
   radius=1.005
   shapeName=dyn_bootZipper_r_capsule
   center=<-0.0096 -0.6075 -1.5853>
   rotation=<rot -0.2786 -90 0.6534>
  }]
}
constraints [
 {bone1Name=dyn_bootZipper_l, bone2Name=calf_twist_01_l, swing1LimitDegrees=15, swing2LimitDegrees=15, twistLimitDegrees=3}
 {bone1Name=dyn_bootZipper_r, bone2Name=calf_twist_01_r, swing1LimitDegrees=15, swing2LimitDegrees=15, twistLimitDegrees=3}
]
