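# Animation Blueprint for the F_MED_StallionAviator cosmetic body that follows the parent mesh and adds secondary motion to its accessories.
# Copied parent pose runs through a deform Control Rig, bone constraints for belt, pouch and jacket zipper, then belt and jacket-zipper RigidBody simulations.
# Boot-zipper and several other belt nodes sit outside the output chain.
# The update event sets gravity-override and crouch, slide and frontend alpha variables, with no visible AnimGraph reads.
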AnimBlueprint /BRCosmetics/Characters/Player/Female/Medium/Bodies/F_MED_StallionAviator/Meshes/F_MED_StallionAviator_AnimBP
target_skeleton=&/F_MED_StallionAviator_Skeleton
graphs [
 AnimGraph {nodes [
   "Constraint - dyn_belt_05 (PPP)"!(ComponentPose, Alpha=0, ConstraintWeights_0=0, ConstraintWeights_1:float=0.8, ConstraintWeights_2:float=0.2) -> (Pose) {
    alphaBoolBlend {blendOption=Linear}
    boneToModify {boneName=dyn_belt_05}
    constraintSetup [{targetBone {boneName=pelvis}}, {targetBone {boneName=thigh_twist_01_l}}, {targetBone {boneName=thigh_l}}]
    constraintWeights [1,1,1]
   }
   n1 = "Constraint - dyn_bootZipper_l (P)"!(ComponentPose, Alpha=1, ConstraintWeights_0=1, ConstraintWeights_1:float=0) -> (Pose) {
    alphaBoolBlend {blendOption=Linear}
    boneToModify {boneName=dyn_bootZipper_l}
    constraintSetup [{targetBone {boneName=calf_twist_01_l}}]
    constraintWeights [1,0]
   }
   "Constraint - dyn_bootZipper_r (P)"!(ComponentPose=@n1, Alpha=1, ConstraintWeights_0=1, ConstraintWeights_1:float=0) -> (Pose) {
    alphaBoolBlend {blendOption=Linear}
    boneToModify {boneName=dyn_bootZipper_r}
    constraintSetup [{targetBone {boneName=calf_twist_01_r}}]
    constraintWeights [1,0]
   }
   "Transform (Modify) Bone - Bone: dyn_belt_02"(Alpha=1, Translation="-0.000,0.000,5.007", Rotation="0.000,0.000,-90.000", Scale="1,1,1") {
    rotationMode=BMM_Additive
    translationMode=BMM_Additive
    alphaBoolBlend {blendOption=Linear}
    boneToModify {boneName=dyn_belt_02}
    rotation {roll=-90}
    translation {z=5.007}
   }
   n2 = "Transform (Modify) Bone - Bone: dyn_belt_02"(Alpha=1, Translation="6.711,-4.360,4.276", Rotation="6.077,137.404,53.074", Scale="1,1,1") {
    rotationMode=BMM_Additive
    translationMode=BMM_Additive
    alphaBoolBlend {blendOption=Linear}
    boneToModify {boneName=dyn_belt_02}
    rotation=<rot 6.077 137.404 53.074>
    translation=<6.711 -4.36 4.276>
   }
   "Transform (Modify) Bone - Bone: dyn_belt_03"(ComponentPose=@n2, Alpha=1, Translation="0,0,0", Rotation="-0.000,-130.000,0.000", Scale="1,1,1") {
    rotationMode=BMM_Additive
    translationMode=BMM_Additive
    alphaBoolBlend {blendOption=Linear}
    boneToModify {boneName=dyn_belt_03}
    rotation {yaw=-130}
   }
   "Transform (Modify) Bone - Bone: dyn_belt_01"(Alpha=1, Translation="0,0,0", Rotation="0.000,10.000,-0.000", Scale="1,1,1") {
    rotationMode=BMM_Additive
    translationMode=BMM_Additive
    alphaBoolBlend {blendOption=Linear}
    boneToModify {boneName=dyn_belt_01}
    rotation {yaw=10}
   }
   RigidBody() {
    bClampLinearTranslationLimitToRefPose=true
    bOverrideWorldGravity=true
    overridePhysicsAsset=/Game/Characters/Player/Female/Medium/Bodies/F_MED_StallionAviator/Meshes/F_MED_StallionAviator_BootZipper_Physics
    simulationSpace=BaseBoneSpace
    alphaBoolBlend {blendOption=Linear}
    baseBoneRef {boneName=root}
    componentLinearAccScale=<0.1 0.1 0.1>
    overrideWorldGravity {z=-980}
   }
   n3 = Control Rig(Source=Copy Pose From Mesh() {bUseAttachedParent=true}) {
    alphaBoolBlend {blendOption=Linear}
    controlRigAssetReference {blueprintRigClass=&/F_MED_StallionAviator_DeformRig_ControlRig.F_MED_StallionAviator_DeformRig_ControlRig_C}
   }
   n4 = "Constraint - dyn_belt_01 (PPPPP)"!(
    ComponentPose=Local To Component(LocalPose=@n3)
    Alpha=1
    ConstraintWeights_0=0.2197
    ConstraintWeights_1:float=0.1146
    ConstraintWeights_2:float=0.0535
    ConstraintWeights_3:float=0.1177
    ConstraintWeights_4:float=0.4946
   ) -> (Pose) {
    alphaBoolBlend {blendOption=Linear}
    boneToModify {boneName=dyn_belt_01}
    constraintSetup [
     {targetBone {boneName=spine_02}}
     {targetBone {boneName=spine_04}}
     {targetBone {boneName=spine_03}}
     {targetBone {boneName=spine_01}}
     {targetBone {boneName=pelvis}}
    ]
    constraintWeights [1,1,1,1,1]
   }
   n5 = "Constraint - dyn_pouch (PPPPP)"!(
    ComponentPose=@n4
    Alpha=1
    ConstraintWeights_0=0
    ConstraintWeights_1:float=0
    ConstraintWeights_2:float=0
    ConstraintWeights_3:float=1
    ConstraintWeights_4:float=0
   ) -> (Pose) {
    alphaBoolBlend {blendOption=Linear}
    boneToModify {boneName=dyn_pouch}
    constraintSetup [
     {targetBone {boneName=spine_02}}
     {targetBone {boneName=spine_04}}
     {targetBone {boneName=spine_03}}
     {targetBone {boneName=spine_01}}
     {targetBone {boneName=pelvis}}
    ]
    constraintWeights [1,1,1,1,1]
   }
   n6 = "Constraint - dyn_jacketZipper (P)"!(ComponentPose=@n5, Alpha=1, ConstraintWeights_0=1, ConstraintWeights_1:float=0) -> (Pose) {
    alphaBoolBlend {blendOption=Linear}
    boneToModify {boneName=dyn_jacketZipper}
    constraintSetup [{targetBone {boneName=spine_05}}]
    constraintWeights [1,0]
   }
   n7 = "Constraint - dyn_belt_03 (PPP)"!(ComponentPose=@n6, Alpha=0, ConstraintWeights_0=0, ConstraintWeights_1:float=0.8, ConstraintWeights_2:float=0.2) -> (Pose) {
    alphaBoolBlend {blendOption=Linear}
    boneToModify {boneName=dyn_belt_03}
    constraintSetup [{targetBone {boneName=pelvis}}, {targetBone {boneName=thigh_twist_01_l}}, {targetBone {boneName=thigh_l}}]
    constraintWeights [1,1,1]
   }
   n8 = RigidBody!(ComponentPose=@n7, Alpha, OverrideWorldGravity:Vector="0,0,-980", ExternalForce) -> (Pose) {
    bClampLinearTranslationLimitToRefPose=true
    bOverrideWorldGravity=true
    overridePhysicsAsset=&/F_MED_StallionAviator_Belt_Physics
    simulationSpace=BaseBoneSpace
    alphaBoolBlend {blendOption=Linear}
    baseBoneRef {boneName=spine_01}
    overrideWorldGravity {z=-980}
   }
   n9 = RigidBody!(ComponentPose=@n8, Alpha, ExternalForce) -> (Pose) {
    bClampLinearTranslationLimitToRefPose=true
    bOverrideWorldGravity=true
    overridePhysicsAsset=&/F_MED_StallionAviator_JacketZipper_Physics
    simulationSpace=BaseBoneSpace
    alphaBoolBlend {blendOption=Linear}
    baseBoneRef {boneName=spine_05}
    componentLinearAccScale=<0.2 0.2 0.2>
    overrideWorldGravity {z=-980}
   }
   Output Pose(Result=Component To Local(ComponentPose=@n9)) {layerGroup=None}
  ]}
 EventGraph {nodes [
   Event BlueprintUpdateAnimation()
   > n1 = Gravity Override State Function DEPRECATED(
    CharacterPartAnimInstance=Self-Reference()
    GameTimeInSeconds=GetGameTimeInSeconds()
    GravityOverrideParams=Get gravity override params() -> (gravity override params:Gravity Override Params Struct Structure)
   )
   > Set gravity override(gravity override:Vector=@n1) -> (Output_Get:Vector)
   > n2 = Branch(Condition=Get bIsCrouching())
   > Set crouching constraint alpha(crouching constraint alpha:double=1) -> (Output_Get:double)
   > Set crouching constraint alpha 2(crouching constraint alpha 2:double=1) -> (Output_Get:double)
   > n3 = Set rban crouch alpha(rban crouch alpha:double=0) -> (Output_Get:double)
   n4 = Branch(execute=[@n6.then, @n3.then], Condition=IsCosmeticObjectDisplayedInFrontendContext!(CosmeticObject:Object Reference) -> (ReturnValue:bool))
   > Set FE Alpha(FE Alpha:double=1) -> (Output_Get:double)
   Set FE Alpha(execute=@n4.else, FE Alpha=1)
   > n5 = Branch(Condition=Get bIsSliding())
   > Set Sliding Alpha(Sliding Alpha:double=0.7) -> (Output_Get:double)
   Set Sliding Alpha(execute=@n5.else, Sliding Alpha=0)
   Set crouching constraint alpha(execute=@n2.else, crouching constraint alpha=0)
   > Set crouching constraint alpha 2(crouching constraint alpha 2=0.2)
   > n6 = Set rban crouch alpha(rban crouch alpha=1)
  ]}
]
variables [
 crouching constraint alpha {type=number}
 crouching constraint alpha 2 {type=number}
 FE Alpha {type=number}
 gravity override {type=Vector}
 gravity override params {type=GravityOverrideParamsStruct}
 rban crouch alpha {type=number}
 Sliding Alpha {type=number}
]
