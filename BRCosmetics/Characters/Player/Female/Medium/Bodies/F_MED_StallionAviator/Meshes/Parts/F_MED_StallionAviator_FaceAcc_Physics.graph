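# Physics asset for the F_MED_StallionAviator face accessory, supporting physical motion of the left and right hat flaps.
# A master hat body is locked to the head and anchors two three-link capsule chains, one per flap.
# Each flap joint has limited motion, widest on swing 1 and tight on swing 2 and twist.
# Head, neck, upper spine and left clavicle capsules supply body shapes, with no right clavicle body.
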PhysicsAsset /BRCosmetics/Characters/Player/Female/Medium/Bodies/F_MED_StallionAviator/Meshes/Parts/F_MED_StallionAviator_FaceAcc_Physics
bodies {
 clavicle_l [{shapeType=Capsule, length=6.25, radius=6.25, shapeName=clavicle_l_capsule, center=<6.5696 0 -1.1483>, rotation=<rot -90 180 179.9999>}]
 dyn_hat_MSTR [{shapeType=Capsule, length=0.2377, radius=0.5, shapeName=dyn_hat_MSTR_capsule, rotation=<rot -50 0.0005 -0.0004>}]
 dyn_hatflap_01_l [{
   shapeType=Capsule
   length=1.7897
   radius=1.2412
   shapeName=dyn_hatflap_01_l_capsule
   center=<2.7207 0.8245 -1.4239>
   rotation=<rot -0.101 -91.533 -1.2673>
  }]
 dyn_hatflap_01_r [{
   shapeType=Capsule
   length=1.7897
   radius=1.2412
   shapeName=dyn_hatflap_01_r_capsule
   center=<3.0881 -0.6376 -1.476>
   rotation=<rot -9.9024 -88.4439 -1.5381>
  }]
 dyn_hatflap_02_l [{
   shapeType=Capsule
   length=2.3136
   radius=2.0442
   shapeName=dyn_hatflap_02_l_capsule
   center=<0.1748 0.2491 1.0058>
   rotation=<rot -9.6348 -90 -0.6474>
  }]
 dyn_hatflap_02_r [{
   shapeType=Capsule
   length=2.3147
   radius=2.0438
   shapeName=dyn_hatflap_02_r_capsule
   center=<0.5652 -0.0988 0.9578>
   rotation=<rot -5.1722 -90 -0.7598>
  }]
 dyn_hatflap_03_l [{
   shapeType=Capsule
   length=1.055
   radius=2.5732
   shapeName=dyn_hatflap_03_l_capsule
   center=<-1.7068 -0.0565 0.1662>
   rotation=<rot -8.8917 -90 -3.2031>
  }]
 dyn_hatflap_03_r [{
   shapeType=Capsule
   length=1.8198
   radius=2.9753
   shapeName=dyn_hatflap_03_r_capsule
   center=<-1.066 0.0643 0.3696>
   rotation=<rot -5.7586 -90 0.6812>
  }]
 head [{shapeType=Capsule, length=0.25, radius=7.005, shapeName=head_capsule, center=<2.5842 0 0>, rotation=<rot 0 -90 0>}]
 neck_02 [{shapeType=Capsule, length=9.5, radius=4.75, shapeName=neck_02_capsule, rotation=<rot 90 5.3558 15.3561>}]
 spine_05 [{shapeType=Capsule, length=21.75, radius=7.5, shapeName=spine_05_capsule, center=<1.5462 0 0>}]
}
constraints [
 {
  bone1Name=dyn_hat_MSTR
  bone2Name=head
  swing1LimitDegrees=45
  swing1Motion=Locked
  swing2LimitDegrees=45
  swing2Motion=Locked
  twistLimitDegrees=45
  twistMotion=Locked
 }
 {bone1Name=dyn_hatflap_01_l, bone2Name=dyn_hat_MSTR, swing1LimitDegrees=30, swing2LimitDegrees=5, twistLimitDegrees=3}
 {bone1Name=dyn_hatflap_02_l, bone2Name=dyn_hatflap_01_l, swing1LimitDegrees=30, swing2LimitDegrees=5, twistLimitDegrees=3}
 {bone1Name=dyn_hatflap_03_l, bone2Name=dyn_hatflap_02_l, swing1LimitDegrees=30, swing2LimitDegrees=5, twistLimitDegrees=3}
 {bone1Name=dyn_hatflap_01_r, bone2Name=dyn_hat_MSTR, swing1LimitDegrees=30, swing2LimitDegrees=5, twistLimitDegrees=3}
 {bone1Name=dyn_hatflap_02_r, bone2Name=dyn_hatflap_01_r, swing1LimitDegrees=30, swing2LimitDegrees=5, twistLimitDegrees=3}
 {bone1Name=dyn_hatflap_03_r, bone2Name=dyn_hatflap_02_r, swing1LimitDegrees=30, swing2LimitDegrees=5, twistLimitDegrees=3}
]
