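# Post-process-style Anim Blueprint for the StallionAviator female face accessory, simulating its hat flaps.
# It copies the parent mesh pose, runs head-space RigidBody physics with reduced component acceleration, then applies corrective bone offsets, mostly to the right flap.
# Update logic derives arm-raise, physics-alpha, gravity-override and frontend values from parachute state, upper-arm rotation and display context.
# The AnimGraph shows fixed inputs, not these variables, and two flap nodes sit unconnected.
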AnimBlueprint /BRCosmetics/Characters/Player/Female/Medium/Bodies/F_MED_StallionAviator/Meshes/Parts/F_MED_StallionAviator_FaceAcc_AnimBP
target_skeleton=&/F_MED_StallionAviator_FaceAcc_Skeleton
graphs [
 AnimGraph {nodes [
   n1 = "Transform (Modify) Bone - Bone: dyn_hatflap_01_r"(Alpha=1, Translation="-1,0,0", Rotation="0,0,0", Scale="1,1,1") {translationMode=BMM_Additive, alphaBoolBlend {blendOption=Linear}, boneToModify {boneName=dyn_hatflap_01_r}, translation {x=-1}}
   "Transform (Modify) Bone - Bone: dyn_hatflap_01_l"(ComponentPose=@n1, Alpha=1, Translation="1,0,0", Rotation="0,0,0", Scale="1,1,1") {translationMode=BMM_Additive, alphaBoolBlend {blendOption=Linear}, boneToModify {boneName=dyn_hatflap_01_l}, translation {x=1}}
   n2 = RigidBody!(
    ComponentPose=Local To Component(LocalPose=Copy Pose From Mesh() {bUseAttachedParent=true})
    Alpha=0.9
    OverrideWorldGravity:Vector="0,0,-980"
    ExternalForce
   ) -> (Pose) {
    bClampLinearTranslationLimitToRefPose=true
    bOverrideWorldGravity=true
    overridePhysicsAsset=&/F_MED_StallionAviator_FaceAcc_Physics
    simulationSpace=BaseBoneSpace
    alphaBoolBlend {blendOption=Linear}
    baseBoneRef {boneName=head}
    componentLinearAccScale=<0.1 0.1 0.1>
    overrideWorldGravity {z=-980}
   }
   n3 = "Transform (Modify) Bone - Bone: dyn_hatflap_01_r"(ComponentPose=@n2, Alpha=1, Translation="0,0,0", Rotation="10.000,0.000,0.000", Scale="1,1,1") {
    rotationMode=BMM_Additive
    translationMode=BMM_Additive
    alphaBoolBlend {blendOption=Linear}
    boneToModify {boneName=dyn_hatflap_01_r}
    rotation {pitch=10}
   }
   n4 = "Transform (Modify) Bone - Bone: dyn_hatflap_02_r"(ComponentPose=@n3, Alpha=1, Translation="2,0,0", Rotation="10,0,0", Scale="1,1,1") {
    rotationMode=BMM_Additive
    translationMode=BMM_Additive
    alphaBoolBlend {blendOption=Linear}
    boneToModify {boneName=dyn_hatflap_02_r}
    rotation {pitch=10}
    translation {x=2}
   }
   n5 = "Transform (Modify) Bone - Bone: dyn_hatflap_03_r"(ComponentPose=@n4, Alpha=1, Translation="0,0,0", Rotation="0,0,0", Scale="0.8,0.8,0.8") {
    rotationMode=BMM_Additive
    scaleMode=BMM_Replace
    translationMode=BMM_Additive
    alphaBoolBlend {blendOption=Linear}
    boneToModify {boneName=dyn_hatflap_03_r}
    scale=<0.8 0.8 0.8>
   }
   n6 = "Transform (Modify) Bone - Bone: dyn_hatflap_02_l"(ComponentPose=@n5, Alpha=1, Translation="0,0,0", Rotation="0,0,0", Scale="1,1,1") {rotationMode=BMM_Additive, translationMode=BMM_Additive, alphaBoolBlend {blendOption=Linear}, boneToModify {boneName=dyn_hatflap_02_l}}
   n7 = "Transform (Modify) Bone - Bone: dyn_hatflap_03_r"(ComponentPose=@n6, Alpha=1, Translation="-1,-0.000,1", Rotation="-40,0,-0.000", Scale="0,0,0") {
    rotationMode=BMM_Additive
    translationMode=BMM_Additive
    alphaBoolBlend {blendOption=Linear}
    boneToModify {boneName=dyn_hatflap_03_r}
    rotation {pitch=-40}
    scale=<0 0 0>
    translation {x=-1, z=1}
   }
   Output Pose(Result=Component To Local(ComponentPose=@n7)) {layerGroup=None}
  ]}
 EventGraph {nodes [
   Event BlueprintUpdateAnimation()
   > n1 = Branch(Condition=Get bIsParachuteOpened())
   > Set Arm Up R(Arm Up R:double=0) -> (Output_Get:double)
   > Set Arm Up L(Arm Up L:double=0) -> (Output_Get:double)
   > Set rban alpha(rban alpha:double=0.2) -> (Output_Get:double)
   Set rban alpha(execute=@n1.else, rban alpha=0.9)
   n2 = GetOwnerSkeletalMeshComponent()
   n3 = BreakRotator(InRot=BreakTransform(InTransform=GetDeltaTransformFromRefPose(self=@n2, BoneName=upperarm_r, BaseName=spine_05)).Rotation)
   > Set Arm Up R(Arm Up R=MapRangeClamped(Value=@n3.Yaw, InRangeA=50, InRangeB=180))
   n4 = GetOwnerSkeletalMeshComponent()
   n5 = BreakRotator(InRot=BreakTransform(InTransform=GetDeltaTransformFromRefPose(self=@n4, BoneName=upperarm_l, BaseName=spine_05)).Rotation)
   > Set Arm Up L(Arm Up L=MapRangeClamped(Value="Absolute (Float)"(A=@n5.Yaw), InRangeA=40, InRangeB=100))
   > n6 = Gravity Override State Function DEPRECATED(
    CharacterPartAnimInstance=Self-Reference()
    GameTimeInSeconds=GetGameTimeInSeconds()
    GravityOverrideParams=Get Gravity override params() -> (Gravity override params:Gravity Override Params Struct Structure)
   )
   > Set Gravity Override(Gravity Override=@n6)
   > n7 = Branch(Condition=IsCosmeticObjectDisplayedInFrontendContext!(CosmeticObject:Object Reference) -> (ReturnValue:bool))
   > Set FE Alpha(FE Alpha:double=1) -> (Output_Get:double)
   > Set rban alpha(rban alpha=0)
   > Set Arm Up R(Arm Up R=0)
   > Set Arm Up L(Arm Up L=0)
   Set FE Alpha(execute=@n7.else, FE Alpha=0)
   "Absolute (Float)"(A=@n3.Yaw)
  ]}
]
variables [
 Arm Up L {type=number}
 Arm Up R {type=number}
 FE Alpha {type=number}
 Gravity Override {type=Vector}
 Gravity override params {type=GravityOverrideParamsStruct}
 rban alpha {type=number}
]
